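# Material function that adds a tiled leathery-skin detail layer to incoming material attributes, named for PistolCanary.
# A shared tiling parameter scales UVs for noise-pack color and normal textures; color multiplies into BaseColor.
# The detail normal blends angle-corrected onto the base, its flattening eased from full toward a flatness parameter by input M's green channel.
# An Alpha input blends between the original and detailed attributes.
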MaterialFunction /Hermes/Assets/Materials/MF_PistolCanary_LeatherySkin
graph {nodes [
  n1 = FunctionInput() {inputName=Material, inputType=FunctionInput_MaterialAttributes}
  n2 = GetMaterialAttributes(None=@n1) {attributeGetTypes [#1, #2]}
  n3 = Multiply(A=TextureCoordinate(), B=ScalarParameter() {defaultValue=20.3585, parameterName=DetailNormaltilint})
  n4 = ScalarParameter() {defaultValue=0.8, parameterName=DetailNormalFlatness}
  n5 = MaterialFunctionCall(Float3=FunctionInput() {inputName=M, inputType=FunctionInput_Scalar}) {materialFunction=&BreakOutFloat3Components}
  n6 = MaterialFunctionCall(
   BaseNormal=@n2.Normal
   AdditionalNormal=MaterialFunctionCall(
    Normal=TextureSample(UVs=@n3) {samplerType=SAMPLERTYPE_Normal, texture=/Hermes/Assets/Textures/Noisepack_Detail_N}.RGB
    Flatness=LinearInterpolate(B=@n4, Alpha=@n5.G) {constA=1}
   ) {materialFunction=&FlattenNormal}.Result
  ) {materialFunction=&BlendAngleCorrectedNormals}
  n7 = Multiply(A=@n2.BaseColor, B=TextureSample(UVs=@n3) {texture=/Hermes/Assets/Textures/Noisepack_Detail_C}.RGB)
  FunctionOutput(None=BlendMaterialAttributes(
    A=@n1
    B=SetMaterialAttributes!(MaterialAttributes=@n2.MaterialAttributes, Normal=@n6.Result, BaseColor=@n7) {attributeSetTypes [#1, #2]}
    Alpha=FunctionInput() {inputName=Alpha, inputType=FunctionInput_Scalar}
   ))
 ]}
