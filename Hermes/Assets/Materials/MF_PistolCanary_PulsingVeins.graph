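# Material function for the PistolCanary that layers tinted, glowing, pulsing veins onto input material attributes through a Mask blend.
# Pulsing veins multiply a lightning color by panning noise on UV1 and an inverted Fresnel.
# Fake subsurface scattering scales a color by sky-atmosphere ground illuminance and a sun-facing SphereMask term, using named reroutes.
# The vein layer also overrides shading model, opacity, custom data and emissive, with one scalar driving opacity and CustomData1.
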MaterialFunction /Hermes/Assets/Materials/MF_PistolCanary_PulsingVeins
graph {nodes [
  n1 = FunctionInput() {inputName=Mat, inputType=FunctionInput_MaterialAttributes}
  n2 = GetMaterialAttributes(None=@n1) {attributeGetTypes [#1, #2, #3]}
  n3 = ScalarParameter() {parameterName=456}
  n4 = SetMaterialAttributes!(
   MaterialAttributes=@n2.MaterialAttributes
   ShadingModel=ShadingModel()
   BaseColor=Multiply(A=@n2.BaseColor, B=VectorParameter() {parameterName=VeinTint, defaultValue=<rgba 0.2588 0.129 0.4271 1>}.RGB)
   Opacity=@n3
   CustomData0=Constant() {r=1}
   CustomData1=@n3
   EmissiveColor=Add(A=NamedRerouteUsage(), B=NamedRerouteUsage())
  ) {attributeSetTypes [
    #4
    #1
    #5
    #6
    #7
    #8
   ]}
  FunctionOutput(None=BlendMaterialAttributes(A=@n1, B=@n4, Alpha=FunctionInput() {inputName=Mask, inputType=FunctionInput_Scalar}))
  n5 = VectorParameter() {parameterName=EmissiveLightningColor, defaultValue=<rgba 0.0442 0.2543 0.6667 1>}
  n6 = Multiply(A=TextureCoordinate() {coordinateIndex=1}, B=ScalarParameter() {defaultValue=1, parameterName=GlowTiling})
  n7 = AppendVector(A=Constant(), B=ScalarParameter() {defaultValue=0.186, parameterName=ElectricSpeed})
  n8 = LinearInterpolate(
   A=ScalarParameter() {defaultValue=0.52, parameterName=GlowMin}
   Alpha=TextureSample(UVs=Panner(Coordinate=@n6, Speed=@n7) {speedY=0.4}) {samplerType=SAMPLERTYPE_LinearColor, texture=/Hermes/Assets/Textures/Noisepack_M}.G
  ) {constA=0.5, constB=1.5}
  n9 = Saturate(None=Fresnel(ExponentIn=ScalarParameter() {defaultValue=0.132, parameterName=FesEx_Veins}, Normal=Transform(None=@n2.Normal)) {baseReflectFraction=0, exponent=1})
  NamedRerouteDeclaration(None=Multiply(A=Multiply(A=@n5.RGB, B=@n8), B=OneMinus(None=@n9))) {name=Pulsing Veins, nodeColor=<rgba 0 0.9766 1 1>}
  n10 = Multiply(
   A=VectorParameter() {parameterName=Subsurfaceintensity, defaultValue=<rgba 0.1457 0.0279 0.4115 1>}.RGB
   B=Multiply(A=SkyAtmosphereLightIlluminanceOnGround(), B=ScalarParameter() {defaultValue=0.179, parameterName=VeinSkyContribution})
  )
  n11 = NamedRerouteUsage()
  NamedRerouteDeclaration(None=Multiply(A=Multiply(A=@n10, B=@n11), B=ScalarParameter() {defaultValue=0.2443, parameterName=SubSurfaceintensity})) {name=Fake SSS, nodeColor=<rgba 0.1719 1 0 1>}
  n12 = LinearInterpolate(
   A=ScalarParameter() {defaultValue=0.8285, parameterName=Vein_OffSide_SSS_Intensity}
   B=ScalarParameter() {defaultValue=1.3806, parameterName=Vein_SunSide_SSS_Intensity}
   Alpha=SphereMask(A=SkyAtmosphereLightDirection(), B=PixelNormalWS(), Radius=ScalarParameter() {defaultValue=1.5001, parameterName=Vein_UnLightRadius}) {hardnessPercent=0}
  ) {constA=0.25}
  NamedRerouteDeclaration(None=NamedRerouteDeclaration(None=Saturate(None=@n12)) {name=SunSpecular, nodeColor=<rgba 0 0.7656 1 1>}) {name=Subsurface, nodeColor=<rgba 0.6406 1 0 1>}
 ]}
