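# Material function that adds a colored glow to a material's emissive around a fixed world position, passing its other attributes through.
# A camera-relative sphere mask, raised to the fourth power for steep falloff, sets the glow's reach.
# Squared inverted world-normal Z, scaled and clamped to at least one, boosts downward-facing surfaces.
# Color, radius and position are parameters, as is intensity, which defaults to zero.
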MaterialFunction /Hermes/Assets/Materials/MF_WorldGlow_PistolCanary
graph {nodes [
  n1 = GetMaterialAttributes(None=FunctionInput() {inputType=FunctionInput_MaterialAttributes}) {attributeGetTypes [#1, #2]}
  n2 = OneMinus(None=ComponentMask(None=Transform(None=@n1.Normal)) {b=true})
  n3 = Multiply(A=Multiply(A=@n2, B=@n2), B=ScalarParameter() {defaultValue=1.3291, group=World Glow MF, parameterName=DownVecRoot})
  n4 = VectorParameter() {group=World Glow MF, parameterName=World Glow World Position, defaultValue=<rgba 14828.5957 70427.8672 3875.54 1>}
  n5 = SphereMask(
   A=TransformPosition(None=@n4.RGB) {transformSourceType=TRANSFORMPOSSOURCE_World, transformType=TRANSFORMPOSSOURCE_TranslatedWorld}
   B=WorldPosition() {worldPositionShaderOffset=WPT_CameraRelative}.XYZ
   Radius=ScalarParameter() {defaultValue=14400, group=World Glow MF, parameterName=GlowRadius}
  ) {hardnessPercent=0}
  n6 = Multiply(A=@n5, B=@n5)
  n7 = Multiply(A=@n6, B=@n6)
  n8 = Multiply(
   A=Multiply(A=Clamp(None=@n3) {maxDefault=500, minDefault=1}, B=@n7)
   B=Multiply(
    A=VectorParameter() {group=World Glow MF, parameterName=WorldRootGlow, defaultValue=<rgba 1.3991 0.0338 1.5 1>}.RGB
    B=ScalarParameter() {group=World Glow MF, parameterName=WorldRootGlowIntensity}
   )
  )
  FunctionOutput(None=SetMaterialAttributes!(MaterialAttributes=@n1.MaterialAttributes, EmissiveColor=Add(A=@n1.EmissiveColor, B=@n8)) {attributeSetTypes [#1]})
  Multiply(A=@n7, B=@n7)
  WorldPosition()
 ]}
